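# Physics asset that supplies collision shapes for the Battle Suit cape backpack cosmetic, named for use as cloth collision.
# It has a single body on the backpack_root bone, built from two overlapping capsules of equal radius, rotated 90 degrees and offset from each other mainly along Z.
# There are no constraints, so it defines only a collision volume with no jointed bodies.
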PhysicsAsset /BRCosmetics/Accessories/FORT_Capes/M_MED_BattleSuit_Cape/Meshes/M_MED_Battle_Suit_Pack_Cloth_Collision_Phsyics
bodies {backpack_root [
  {shapeType=Capsule, length=11.1431, radius=12.3508, center=<0.0002 -9.5463 -3.3794>, rotation=<rot 90 0 0>}
  {shapeType=Capsule, length=10.1615, radius=12.3508, shapeName=backpack_root_capsule, center=<0.2058 -8.0703 3.4079>, rotation=<rot 90 0 0>}
 ]}
constraints []
